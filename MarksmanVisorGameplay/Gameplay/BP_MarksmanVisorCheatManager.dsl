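; Cheat manager Blueprint for the Marksman Visor, exposing debug functions that override its footstep and trail settings.
; Four functions set either a duration override or an active override for footsteps or trails.
; Each finds the Marksman Visor component on the Athena game state and writes the value only if both are valid.
; The active overrides take a bool-with-unset enum rather than a plain boolean.
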(fn CheatVisorFootstepDuration
  (params (FootstepDurationOverride :type Float%20%28double-precision%29))
  (bind _returnvalue (Fortnite|Athena|GameState|GetGameStateAthena))
  (bind _returnvalue_2 (Actor|GetComponentbyClass _returnvalue "/MarksmanVisorGameplay/Gameplay/B_MarksmanVisor_GameStateComponent.B_MarksmanVisor_GameStateComponent_C"))
  (Utilities|IsValid _returnvalue
    (:Is%20Valid
      (Utilities|IsValid _returnvalue_2
        (:Is%20Valid
          (Variables|/MarksmanVisorGameplay/Gameplay/B_MarksmanVisor_GameStateComponent.B_MarksmanVisor_GameStateComponent_C|SetFootstepDurationOverride FootstepDurationOverride _returnvalue_2))))))

(fn CheatVisorFootstepEnable
  (params (FootstepActiveOverride :type EBoolWithUnset%20Enum :default "NewEnumerator0"))
  (bind _returnvalue (Fortnite|Athena|GameState|GetGameStateAthena))
  (bind _returnvalue_2 (Actor|GetComponentbyClass _returnvalue "/MarksmanVisorGameplay/Gameplay/B_MarksmanVisor_GameStateComponent.B_MarksmanVisor_GameStateComponent_C"))
  (Utilities|IsValid _returnvalue
    (:Is%20Valid
      (Utilities|IsValid _returnvalue_2
        (:Is%20Valid
          (Variables|/MarksmanVisorGameplay/Gameplay/B_MarksmanVisor_GameStateComponent.B_MarksmanVisor_GameStateComponent_C|SetFootstepActiveOverride FootstepActiveOverride _returnvalue_2))))))

(fn CheatVisorTrailDuration
  (params (TrailDurationOverride :type Float%20%28double-precision%29))
  (bind _returnvalue (Fortnite|Athena|GameState|GetGameStateAthena))
  (bind _returnvalue_2 (Actor|GetComponentbyClass _returnvalue "/MarksmanVisorGameplay/Gameplay/B_MarksmanVisor_GameStateComponent.B_MarksmanVisor_GameStateComponent_C"))
  (Utilities|IsValid _returnvalue
    (:Is%20Valid
      (Utilities|IsValid _returnvalue_2
        (:Is%20Valid
          (Variables|/MarksmanVisorGameplay/Gameplay/B_MarksmanVisor_GameStateComponent.B_MarksmanVisor_GameStateComponent_C|SetTrailDurationOverride TrailDurationOverride _returnvalue_2))))))

(fn CheatVisorTrailEnable
  (params (TrailActiveOverride :type EBoolWithUnset%20Enum :default "NewEnumerator0"))
  (bind _returnvalue (Fortnite|Athena|GameState|GetGameStateAthena))
  (bind _returnvalue_2 (Actor|GetComponentbyClass _returnvalue "/MarksmanVisorGameplay/Gameplay/B_MarksmanVisor_GameStateComponent.B_MarksmanVisor_GameStateComponent_C"))
  (Utilities|IsValid _returnvalue
    (:Is%20Valid
      (Utilities|IsValid _returnvalue_2
        (:Is%20Valid
          (Variables|/MarksmanVisorGameplay/Gameplay/B_MarksmanVisor_GameStateComponent.B_MarksmanVisor_GameStateComponent_C|SetTrailActiveOverride TrailActiveOverride _returnvalue_2))))))
