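; Character part modifier for the Marksman Visor that animates its materials and toggles its visibility.
; On post-resource init it caches the part's skeletal mesh and plays a timeline driving the deploy parameter (spelled "Depoly_Alpha").
; On authority only, async tag tasks run targeting timelines on IsTargeting count changes and hide the mesh while BeastGirl or HideVisuals tags are present.
; The stop animation skips zero-value updates, and EndPlay ends only the show-again task.
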(event EventOnPostResIn
  (bind _returnvalue (CustomCharacter|GetSkeletalMeshForPartOnPlayerPawn))
  (|CallParentFunction)
  (Variables|Default|SetSkeletalMesh _returnvalue)
  (Rendering|Material|SetScalarParameterValueonMaterials _returnvalue "IsTargeting" 0.0)
  (bind (_direction _deployalpha) (Timeline|DeployAnimation PlayFromStart)
    (:Update
      (Rendering|Material|SetScalarParameterValueonMaterials (Variables|Default|GetSkeletalMesh) "Depoly_Alpha" _deployalpha))))

(event Custom|OnEndPlay_Event
  (params (Actor :type Actor%20Object%20Reference) (EndPlayReason :type EEndPlayReason%20Enum :default "Destroyed"))
  (Ability|Async|EndAction (Variables|Default|GetAsync_ListenFalse)))

(event EventBeginPlay
  (bind _returnvalue (Actor|GetOwner))
  (Networking|SwitchHasAuthority
    (:Authority
      (bind _asfort_player_pawn (Utilities|Casting|CastToFortPlayerPawn (Actor|GetOwner)))
      (bind (_asynctaskproxy _tagcount) (Ability|Tasks|WaitGameplayTagCountChangedOnActor _asfort_player_pawn "(TagName=\"Gameplay.Status.IsTargeting\")")
        (:then
          (bind _asynctaskproxy_2 (Ability|Tasks|WaitGameplayTagQueryOnActor _returnvalue "(TokenStreamVersion=0,TagDictionary=((TagName=\"Granted.BeastGirl.Pawn\"),(TagName=\"Granted.Athena.Item.MarksmanVisor.HideVisuals\")),QueryTokenStream=(0,1,1,2,0,1),UserDescription=\"\",AutoDescription=\" ANY( Granted.BeastGirl.Pawn, Granted.Athena.Item.MarksmanVisor.HideVisuals )\")")
            (:then
              (Variables|Default|SetAsync_ListenTrue _asynctaskproxy_2)
              (bind _asynctaskproxy_3 (Ability|Tasks|WaitGameplayTagQueryOnActor _returnvalue "(TokenStreamVersion=0,TagDictionary=((TagName=\"Granted.BeastGirl.Pawn\"),(TagName=\"Granted.Athena.Item.MarksmanVisor.HideVisuals\")),QueryTokenStream=(0,1,1,2,0,1),UserDescription=\"\",AutoDescription=\" ANY( Granted.BeastGirl.Pawn, Granted.Athena.Item.MarksmanVisor.HideVisuals )\")" "WhenFalse")
                (:then
                  (Variables|Default|SetAsync_ListenFalse _asynctaskproxy_3)
                  (Game|BindEventtoOnEndPlay (delegate Custom|OnEndPlay_Event)))
                (:Triggered
                  (Rendering|SetVisibility (Variables|Default|GetSkeletalMesh) true false))))
            (:Triggered
              (Rendering|SetVisibility (Variables|Default|GetSkeletalMesh) false false))))
        (:TagCountChanged
          (if (>= _tagcount 1)
            (CallFunction|DoTargetingAnim)
            (else
              (CallFunction|StopTargetingAnim))))))))

(event Custom|DoTargetingAnim
  (bind (_direction _targetingalpha) (Timeline|TargetingAnimation PlayFromStart)
    (:Update
      (Rendering|Material|SetScalarParameterValueonMaterials (Variables|Default|GetSkeletalMesh) "IsTargeting" _targetingalpha))))

(event Custom|StopTargetingAnim
  (bind (_direction _targetingalpha) (Timeline|Timeline ReverseFromEnd)
    (:Update
      (if (not (== _targetingalpha 0))
        (Rendering|Material|SetScalarParameterValueonMaterials (Variables|Default|GetSkeletalMesh) "IsTargeting" _targetingalpha)))))

(fn ConstructionScript
  (|CallParentFunction))
