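; Pawn component for the Marksman Visor that shows footstep trails, a highlighted footstep and marked pawns to the local viewer.
; A replicated toggle plus view-target checks start footstep generation, an owner-only Niagara system and a highlight actor showing scan progress and cooldown.
; The authority keeps capped, timed pawn marks whose IDs feed Niagara parameters, refreshed by timer at the next expiry.
; Dead generators' IDs are queued and fed to Niagara one at a time.
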(event Highlight|EventSetHighlightFootstep
  (params (Footstep :type Visualized%20Footstep%20Data%20Structure%20%28by%20ref%29))
  ; Highlight Footstep
  (Variables|FootstepHighlight|SetHighlightFootstep Footstep)
  (Variables|FootstepHighlight|SetHighlightFootstepValid true)
  (CallFunction|UpdateHighlightFootstep))

(event EventBeginPlay
  ; Begin Play
  (if (not (Networking|IsDedicatedServer))
    (Variables|Components|SetFortPawnOwner (Utilities|Casting|CastToFortPawn (Components|GetOwner)))))

(event Highlight|EventClearHighlightFootstep
  ; Highlight Footstep
  (Variables|FootstepHighlight|SetHighlightFootstepValid false)
  (CallFunction|UpdateHighlightFootstep))

(event Custom|MarkPawn
  (params (PawnToMark :type Fort%20Pawn%20Object%20Reference) (Duration :type Float%20%28double-precision%29))
  (bind _activelymarkedpawns (Variables|Default|GetActivelyMarkedPawns))
  ; Mark a Pawn
  (Utilities|SwitchHasAuthority
    (:Authority
      (Utilities|IsValid PawnToMark
        (:Is%20Valid
          (Utilities|Array|Add (Variables|Default|GetActivelyMarkedPawns) (Utilities|Struct|MakeSPawnIdandTimePair (Variables|/Script/FortniteGame.FortPawn|GetPawnUniqueID PawnToMark) (+ Duration (Utilities|Time|GetTimeSeconds))))
          (Utilities|FlowControl|Sequence
            (:then_0
              ;; Too many marked pawns?
              (if (> (Utilities|Array|Length _activelymarkedpawns) (Utilities|Array|Length (Variables|Default|GetNiagaraTrackedTargetParamNames)))
                (Utilities|Array|RemoveIndex _activelymarkedpawns)))
            (:then_1
              (CallFunction|RefreshMarkedPawns))))))))

(event Custom|SetHighlightFootstepActiveScanning
  (params (IsScanning :type Boolean) (ProgressPercent :type Float%20%28double-precision%29))
  ; Highlight Footstep
  (Variables|FootstepHighlight|SetHighlightFootstepIsScanning IsScanning)
  (Variables|FootstepHighlight|SetHighlightFootstepProgressPercent ProgressPercent)
  (CallFunction|UpdateHighlightFootstep))

(event EventEndPlay
  (params (EndPlayReason :type EEndPlayReason%20Enum :default "Destroyed"))
  ; End Play
  (switch Utilities|FlowControl|Switch|SwitchonEEndPlayReason EndPlayReason
    (:Destroyed
      _)
    (:LevelTransition
      _)
    (:RemovedFromWorld
      _))
  (Variables|Default|SetVisualizeFootsteps false)
  (Utilities|Time|ClearandInvalidateTimerbyHandle (Variables|Default|GetRefreshMarkedPawnsHandle))
  (bind _highlightvisualizationactor (Variables|FootstepHighlight|GetHighlightVisualizationActor)
    (:then
      (Actor|DestroyActor _highlightvisualizationactor)
      (Variables|FootstepHighlight|SetHighlightVisualizationActor 0))
    (:else)))

(event Visualization|EventBeginFootstepVisualization
  ; Footstep Visualization
  (if (not (Variables|Default|GetVisualizeFootsteps))
    (Variables|Default|SetVisualizeFootsteps true)))

(event Visualization|EventEndFootstepVisualization
  ; Footstep Visualization
  (if (Variables|Default|GetVisualizeFootsteps)
    (Variables|Default|SetVisualizeFootsteps false)))

(event Custom|OnRep_VisualizeFootsteps
  ; Footstep Visualization
  (if (and (Utilities|IsValid (Variables|Components|GetFortPawnOwner) :!class /Script/Engine.KismetSystemLibrary) (not (Networking|IsDedicatedServer)))
    (CallFunction|LazyCaptureGameStateComponent)
    (bind _visorgamestatecomponent (Variables|Default|GetVisorGameStateComponent)
      (:then
        (Variables|Default|GetVisualizeFootsteps
          (:then
            (EventDispatchers|BindEventtoOnVisorLocalViewTargetPawnsChanged _visorgamestatecomponent (delegate Custom|LocalViewTargetsChanged)))
          (:else
            (EventDispatchers|UnbindEventfromOnVisorLocalViewTargetPawnsChanged _visorgamestatecomponent (delegate Custom|LocalViewTargetsChanged))))
        (bind _visorgamestatecomponent_2 (Variables|Default|GetVisorGameStateComponent)
          (:then
            (if (and (Class|FortGameStateComponentMarksmanVisor|IsPawnVisorLocalViewTarget _visorgamestatecomponent_2 (Variables|Components|GetFortPawnOwner)) (Variables|Default|GetVisualizeFootsteps))
              (if (not (Variables|Default|GetIsVisualizingFootsteps_Local))
                (CallFunction|BeginFootstepVisualizationInternal))
              (else
                (if (Variables|Default|GetIsVisualizingFootsteps_Local)
                  (CallFunction|EndFootstepVisualizationInternal)))))
          (:else)))
      (:else
        (Development|PrintString "Marksman Visor - missing VisorGameStateComponent. Footstep visualization will fail." true true "(R=1.000000,G=0.016018,B=0.003985,A=1.000000)" 3e+01)))))

(event Custom|LocalViewTargetsChanged
  ; Footstep Visualization
  (bind _visorgamestatecomponent (Variables|Default|GetVisorGameStateComponent)
    (:then
      (if (and (Class|FortGameStateComponentMarksmanVisor|IsPawnVisorLocalViewTarget _visorgamestatecomponent (Variables|Components|GetFortPawnOwner)) (Variables|Default|GetVisualizeFootsteps))
        (if (not (Variables|Default|GetIsVisualizingFootsteps_Local))
          (CallFunction|BeginFootstepVisualizationInternal))
        (else
          (if (Variables|Default|GetIsVisualizingFootsteps_Local)
            (CallFunction|EndFootstepVisualizationInternal)))))
    (:else)))

(event Custom|BeginFootstepVisualization_Internal
  ; Footstep Visualization
  (Variables|Default|SetIsVisualizingFootsteps_Local true)
  (Utilities|FlowControl|Sequence
    (:then_0
      (bind _visorgamestatecomponent (Variables|Default|GetVisorGameStateComponent)
        (:then
          (Class|FortGameStateComponentFootstepVisualization|StartGeneratingFootsteps _visorgamestatecomponent)
          (Default|BindEventtoOverridesModified _visorgamestatecomponent (EventDispatchers|CreateEvent :!event RefreshVisualizationGameplayValues))
          (EventDispatchers|BindEventtoOnFootstepGeneratorDeath _visorgamestatecomponent (EventDispatchers|CreateEvent :!event HandleFootstepGeneratorDeath)))
        (:else)))
    (:then_1
      (CallFunction|BuildVisualization))))

(event Custom|EndFootstepVisualization_Internal
  ; Footstep Visualization
  (Variables|Default|SetIsVisualizingFootsteps_Local false)
  (Utilities|FlowControl|Sequence
    (:then_0
      (bind _visorgamestatecomponent (Variables|Default|GetVisorGameStateComponent)
        (:then
          (Class|FortGameStateComponentFootstepVisualization|EndGeneratingFootsteps _visorgamestatecomponent)
          (Default|UnbindEventfromOverridesModified _visorgamestatecomponent (EventDispatchers|CreateEvent :!event RefreshVisualizationGameplayValues))
          (EventDispatchers|UnbindEventfromOnFootstepGeneratorDeath _visorgamestatecomponent (EventDispatchers|CreateEvent :!event HandleFootstepGeneratorDeath))
          (Variables|Default|SetVisorGameStateComponent 0))
        (:else)))
    (:then_1
      (bind _visualizationinstance (Variables|Default|GetVisualizationInstance)
        (:then
          (Components|Activation|Deactivate _visualizationinstance)
          (Variables|Default|SetVisualizationInstance 0))
        (:else)))
    (:then_2
      (Utilities|Time|ClearandInvalidateTimerbyHandle (Variables|Default|GetDeathRemovalUpdateHandle)))))

(fn BuildVisualization
  (Utilities|FlowControl|Sequence
    (:then_0
      ; Footsteps
      (Variables|Default|GetVisualizationInstance
        (:then)
        (:else
          (bind _asfort_pawn (Utilities|Casting|CastToFortPawn (Components|GetOwner)))
          (bind _returnvalue (Niagara|SpawnSystemAttached "/MarksmanVisorGameplay/Assets/FX/NS_NDC_Handler_MarksmanVisor.NS_NDC_Handler_MarksmanVisor" (Variables|/Script/Engine.Actor|GetRootComponent _asfort_pawn) "None" "0, 0, 0" "0, 0, 0" "SnapToTarget"))
          (bind _output_get (Variables|Default|SetVisualizationInstance _returnvalue))
          (Rendering|SetOnlyOwnerSee _output_get true :!class /Script/Engine.PrimitiveComponent)
          (Niagara|SetNiagaraVariable(Int32) _output_get "InstigatorAffiliation" (Math|Conversions|ToInteger(Byte) (Team|GetActorTeam _asfort_pawn)))
          (CallFunction|RefreshVisualizationGameplayValues)
          (CallFunction|RefreshVisualizationMarkedIDs))))
    (:then_1
      ; Footstep highlight
      (Variables|FootstepHighlight|GetHighlightVisualizationActor
        (:then)
        (:else
          (bind _returnvalue_2 (Game|SpawnActorfromClass "/MarksmanVisorGameplay/Gameplay/Actors/B_MarksmanVisorFootstepHighlightActor.B_MarksmanVisorFootstepHighlightActor_C" (Math|Transform|MakeTransform "0, 0, 0") "Undefined" "MultiplyWithRoot" (Components|GetOwner) 0))
          (Variables|FootstepHighlight|SetHighlightVisualizationActor _returnvalue_2)
          (CallFunction|UpdateHighlightFootstep))))))

(fn DeathRemovalUpdate
  (bind _deathremovalids (Variables|Default|GetDeathRemovalIDs))
  (Utilities|Time|ClearandInvalidateTimerbyHandle (Variables|Default|GetDeathRemovalUpdateHandle))
  ; Pop Next Death ID
  (if (Utilities|Array|IsValidIndex _deathremovalids)
    (LocalVariables|SetCurrentDeathID (Utilities|Array|Get(acopy) _deathremovalids))
    (Utilities|Array|RemoveIndex _deathremovalids)
    ; Set Death ID
    (bind _visualizationinstance (Variables|Default|GetVisualizationInstance)
      (:then
        (Niagara|SetNiagaraVariable(Int32) _visualizationinstance "KillUniqueID" (LocalVariables|GetCurrentDeathID))
        ; Set Next Update Timer
        (bind _returnvalue (Utilities|Time|SetTimerbyFunctionName 0 "" (ScalableFloat|GetValueatLevel (Variables|Default|GetSecondsBetweenDeathRemovals))))
        (Variables|Default|SetDeathRemovalUpdateHandle _returnvalue))
      (:else))
    (else
      ; Clear Death ID
      (bind _visualizationinstance_2 (Variables|Default|GetVisualizationInstance)
        (:then
          (Niagara|SetNiagaraVariable(Int32) _visualizationinstance_2 "KillUniqueID" -1))
        (:else)))))

(fn GetHighlightFootstep
  (return (Variables|FootstepHighlight|GetHighlightFootstep) (Variables|FootstepHighlight|GetHighlightFootstepValid)))

(fn HandleFootstepGeneratorDeath
  (params (DeadPawn :type Fort%20Pawn%20Object%20Reference))
  (Utilities|Array|Add (Variables|Default|GetDeathRemovalIDs) (Variables|/Script/FortniteGame.FortPawn|GetPawnUniqueID DeadPawn))
  (if (not (Utilities|Time|IsValidTimerHandle (Variables|Default|GetDeathRemovalUpdateHandle)))
    (CallFunction|DeathRemovalUpdate)))

(fn LazyCaptureGameStateComponent
  (bind _returnvalue (Game|GetGameState))
  (Variables|Default|GetVisorGameStateComponent
    (:then)
    (:else
      (Utilities|IsValid _returnvalue
        (:Is%20Valid
          (Variables|Default|SetVisorGameStateComponent (Actor|GetComponentbyClass _returnvalue "/MarksmanVisorGameplay/Gameplay/B_MarksmanVisor_GameStateComponent.B_MarksmanVisor_GameStateComponent_C")))))))

(fn OnRep_MarkedIDs
  (CallFunction|RefreshVisualizationMarkedIDs))

(fn RefreshMarkedPawns
  (Utilities|Array|Clear (Variables|Default|GetMarkedIDs))
  (LocalVariables|SetCurrentTime (Utilities|Time|GetTimeSeconds))
  (LocalVariables|SetNextRefreshTime (LocalVariables|GetInvalidRefreshTime))
  (bind (_arrayindex _arrayelement) (Utilities|Array|ReverseforEachLoop (Variables|Default|GetActivelyMarkedPawns))
    (:LoopBody
      (if (<= (.Time_5_F240D04043B5038E53C9DE8A398D86FE (Utilities|Struct|BreakSPawnIdandTimePair _arrayelement)) (LocalVariables|GetCurrentTime))
        (Utilities|Array|RemoveIndex (Variables|Default|GetActivelyMarkedPawns) _arrayindex)
        (else
          (LocalVariables|SetNextRefreshTime (Math|Float|Min(Float) (LocalVariables|GetNextRefreshTime) (.Time_5_F240D04043B5038E53C9DE8A398D86FE (Utilities|Struct|BreakSPawnIdandTimePair _arrayelement))))
          (Utilities|Array|Add (Variables|Default|GetMarkedIDs) (Utilities|Struct|BreakSPawnIdandTimePair _arrayelement)))))
    (:Completed
      ; Set Next Timer
      (if (< (LocalVariables|GetNextRefreshTime) (LocalVariables|GetInvalidRefreshTime))
        (bind _returnvalue (Utilities|Time|SetTimerbyFunctionName 0 "RefreshMarkedPawns" (- (LocalVariables|GetNextRefreshTime) (LocalVariables|GetCurrentTime))))
        (Variables|Default|SetRefreshMarkedPawnsHandle _returnvalue)))))

(fn RefreshVisualizationGameplayValues
  (bind _visualizationinstance (Variables|Default|GetVisualizationInstance)
    (:then
      (Utilities|FlowControl|Sequence
        (:then_0
          ; Default values from curve table
          (LocalVariables|SetFootstepDurationValue (ScalableFloat|GetValueatLevel (Variables|Footstep|GetFootstepDuration)))
          (LocalVariables|SetAreFootstepsEnabledValue (ScalableFloat|GetValueasBool (Variables|Footstep|GetAreFootstepsEnabled))))
        (:then_1
          ; Overrides (cheat values)
          (bind _asb_marksman_visor_game_state_component (Utilities|Casting|CastToB_MarksmanVisor_GameStateComponent (Variables|FortPawnComponent_FootstepVisualization|GetGameStateComponent)))
          (Utilities|FlowControl|Sequence
            (:then_0
              (switch Utilities|FlowControl|Switch|SwitchonEBoolWithUnset (Variables|/MarksmanVisorGameplay/Gameplay/B_MarksmanVisor_GameStateComponent.B_MarksmanVisor_GameStateComponent_C|GetFootstepActiveOverride _asb_marksman_visor_game_state_component)
                (:NewEnumerator1
                  (LocalVariables|SetAreFootstepsEnabledValue false))
                (:NewEnumerator2
                  (LocalVariables|SetAreFootstepsEnabledValue true))))
            (:then_1
              (if (> (Variables|/MarksmanVisorGameplay/Gameplay/B_MarksmanVisor_GameStateComponent.B_MarksmanVisor_GameStateComponent_C|GetFootstepDurationOverride _asb_marksman_visor_game_state_component) 0)
                (LocalVariables|SetFootstepDurationValue (Variables|/MarksmanVisorGameplay/Gameplay/B_MarksmanVisor_GameStateComponent.B_MarksmanVisor_GameStateComponent_C|GetFootstepDurationOverride _asb_marksman_visor_game_state_component))))))
        (:then_2
          ; Push values to Niagara system
          (Niagara|SetNiagaraVariable(Float) _visualizationinstance "FootstepDuration" (LocalVariables|GetFootstepDurationValue))
          (Effects|Components|ParticleSystem|SetEmitterEnable _visualizationinstance "Foot_Step" (LocalVariables|GetAreFootstepsEnabledValue))
          (Niagara|SetNiagaraVariableByString(Int32) _visualizationinstance "CC_StepsPerTrail" (ScalableFloat|GetValueasInteger (Variables|Default|GetWaveformStepsPerTrail)))
          (Niagara|SetNiagaraVariable(Float) _visualizationinstance "CC_LengthScale" (ScalableFloat|GetValueatLevel (Variables|Default|GetWaveformWavesPerTrail)))
          (Niagara|SetNiagaraVariable(Float) _visualizationinstance "CC_ScaleTime" (ScalableFloat|GetValueatLevel (Variables|Default|GetWaveformPanningSpeedScale)))
          (Niagara|SetNiagaraVariable(Float) _visualizationinstance "CC_Frequency" (ScalableFloat|GetValueatLevel (Variables|Default|GetWaveformFrequencyScale)))
          (Niagara|SetNiagaraVariable(Float) _visualizationinstance "CC_Exponent" (ScalableFloat|GetValueatLevel (Variables|Default|GetWaveformExponent)))
          (Niagara|SetNiagaraVariable(Float) _visualizationinstance "CC_LeadingEdgeSeconds" (ScalableFloat|GetValueatLevel (Variables|Default|GetLeadingEdgeSeconds)))
          (Niagara|SetNiagaraVariable(Float) _visualizationinstance "CC_LeadingEdgeHeight" (ScalableFloat|GetValueatLevel (Variables|Default|GetLeadingEdgeHeight))))))
    (:else)))

(fn RefreshVisualizationMarkedIDs
  (Variables|Default|GetVisualizationInstance
    (:then
      ; Supply Niagara with tracked IDs or -1
      (for (_array_element _array_index) (Variables|Default|GetNiagaraTrackedTargetParamNames)
        (LocalVariables|SetCurrentVariableName _array_element)
        (LocalVariables|SetCurrentIndex _array_index)
        (if (Utilities|Array|IsValidIndex (Variables|Default|GetMarkedIDs) (LocalVariables|GetCurrentIndex))
          (Niagara|SetNiagaraVariable(Int32) (Variables|Default|GetVisualizationInstance) (LocalVariables|GetCurrentVariableName) (Utilities|Array|Get(acopy) (Variables|Default|GetMarkedIDs) (LocalVariables|GetCurrentIndex)))
          (else
            (Niagara|SetNiagaraVariable(Int32) (Variables|Default|GetVisualizationInstance) (LocalVariables|GetCurrentVariableName) -1)))))
    (:else)))

(fn UpdateHighlightFootstep
  (bind _highlightfootstep_location (Utilities|Struct|BreakVisualizedFootstepData (Variables|GetHighlightFootstep_Location)))
  (bind _highlightvisualizationactor (Variables|FootstepHighlight|GetHighlightVisualizationActor)
    (:then
      (if (Variables|FootstepHighlight|GetHighlightFootstepValid)
        (Transformation|SetActorLocation _highlightvisualizationactor _highlightfootstep_location false true)
        (bind _returnvalue (Ability|Cooldowns|IsCooldownTagsActive (Components|GetOwner) (Variables|Default|GetScanningCooldownTag)))
        (Class|BMarksmanVisorFootstepHighlightActor|SetState _highlightvisualizationactor true _returnvalue (Variables|FootstepHighlight|GetHighlightFootstepIsScanning) (Variables|FootstepHighlight|GetHighlightFootstepProgressPercent))
        (else
          (Class|BMarksmanVisorFootstepHighlightActor|SetState _highlightvisualizationactor false false false 0.0))))
    (:else)))
