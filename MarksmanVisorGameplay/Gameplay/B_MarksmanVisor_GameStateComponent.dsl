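; Game state component for the Marksman Visor that announces changes to overrides of footstep and trail active states and durations.
; Each of the four override variables has a RepNotify function that calls the OverridesModified event dispatcher.
; All four notifies share that one parameterless dispatcher, so listeners receive a generic change signal without the changed value.
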(event Custom|OnRep_FootstepDurationOverride
  (Default|CallOverridesModified))

(event Custom|OnRep_TrailActiveOverride
  (Default|CallOverridesModified))

(event Custom|OnRep_FootstepActiveOverride
  (Default|CallOverridesModified))

(event Custom|OnRep_TrailDurationOverride
  (Default|CallOverridesModified))
